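# Physics asset for the Number6 vehicle cosmetic body, likely a Battle Royale variant given the BR suffix.
# It defines a single body on the root bone with no constraints, so the whole mesh simulates as one rigid piece.
# The listing shows no collision shapes on that body.
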PhysicsAsset /VehicleCosmetics/Bodies/Number6/Meshes/Body_Number6_PhysicsBR
bodies {root []}
constraints []
